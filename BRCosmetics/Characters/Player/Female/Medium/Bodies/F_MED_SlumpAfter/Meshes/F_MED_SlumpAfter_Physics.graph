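# Physics asset for the F_MED_SlumpAfter female medium body cosmetic, covering secondary motion bones for a bookstrap and left and right skirt flaps.
# Capsules cover the spine, pelvis, legs and feet, with no arms or head, plus an oversized offset capsule on the root.
# The bookstrap chain hangs from the pelvis and each skirt flap chain from its thigh through a free-rotating MSTR body, with tight swing and twist limits along each chain.
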PhysicsAsset /BRCosmetics/Characters/Player/Female/Medium/Bodies/F_MED_SlumpAfter/Meshes/F_MED_SlumpAfter_Physics
bodies {
 calf_l [{shapeType=Capsule, length=36, radius=6.5, center=<-22 -2 0>, rotation=<rot -90 0 0>}]
 calf_r [{shapeType=Capsule, length=36, radius=6.5, center=<22 2 0>, rotation=<rot -90 0 0>}]
 dyn_bookstrap_01 [{shapeType=Capsule, length=3, radius=3, center=<3 0 2.271>, rotation=<rot 90 0 0>}]
 dyn_bookstrap_02 [{shapeType=Capsule, length=3, radius=4.7869, shapeName=dyn_bookstrap_02_capsule, center=<3 0 3.815>, rotation=<rot 90 0 0>}]
 dyn_bookstrap_03 [{
   shapeType=Capsule
   length=10
   radius=12
   shapeName=dyn_bookstrap_03_capsule
   center=<6.965 0.0022 8.3807>
   rotation=<rot -3.8408 179.4626 -90>
  }]
 dyn_bookstrap_MSTR [{shapeType=Capsule, length=0.505, radius=0.505, shapeName=dyn_bookstrap_MSTR_capsule, rotation=<rot 0 -90 0>}]
 dyn_skirtflap_l_01 [{shapeType=Capsule, length=10, radius=8, center=<6 0 4>, rotation=<rot 0 0 90>}]
 dyn_skirtflap_l_02 [{shapeType=Capsule, length=10, radius=8, shapeName=dyn_skirtflap_l_02_capsule, center=<6 0 4>, rotation=<rot 0 0 90>}]
 dyn_skirtflap_l_03 [{shapeType=Capsule, length=10, radius=8, center=<6 0 4>, rotation=<rot -0.0002 0.0001 100>}]
 dyn_skirtflap_l_04 [{shapeType=Capsule, length=10, radius=8, shapeName=dyn_skirtflap_l_04_capsule, center=<6 0 4>, rotation=<rot -0.0003 0.0001 99.9999>}]
 dyn_skirtflap_l_05 [{shapeType=Capsule, length=10, radius=8, center=<6 0 5>, rotation=<rot -0.0001 0 95>}]
 dyn_skirtflap_l_06 [{shapeType=Capsule, length=10, radius=8, shapeName=dyn_skirtflap_l_06_capsule, center=<6 0 6>, rotation=<rot -0.0001 0 95>}]
 dyn_skirtflap_l_MSTR [{shapeType=Capsule, length=0.505, radius=0.505, shapeName=dyn_skirtflap_l_MSTR_capsule, rotation=<rot 0 -90 0>}]
 dyn_skirtflap_r_01 [{shapeType=Capsule, length=10, radius=8, center=<6 0 4>, rotation=<rot 0 0 90>}]
 dyn_skirtflap_r_02 [{shapeType=Capsule, length=10, radius=8, shapeName=dyn_skirtflap_r_02_capsule, center=<6 0 4>, rotation=<rot 0 0 90>}]
 dyn_skirtflap_r_03 [{shapeType=Capsule, length=10, radius=8, center=<6 0 4>, rotation=<rot -0.0001 0 80>}]
 dyn_skirtflap_r_04 [{shapeType=Capsule, length=10, radius=8, shapeName=dyn_skirtflap_r_04_capsule, center=<6 0 4>, rotation=<rot -0.0001 0 80>}]
 dyn_skirtflap_r_05 [{shapeType=Capsule, length=10, radius=8, center=<6 0 5>, rotation=<rot -0.0001 0 85>}]
 dyn_skirtflap_r_06 [{shapeType=Capsule, length=10, radius=8, shapeName=dyn_skirtflap_r_06_capsule, center=<6 0 6>, rotation=<rot -0.0001 0 85>}]
 dyn_skirtflap_r_MSTR [{shapeType=Capsule, length=0.505, radius=0.505, shapeName=dyn_skirtflap_r_MSTR_capsule, rotation=<rot 0 -90 0>}]
 foot_l [{shapeType=Capsule, length=14, radius=5.5, shapeName=foot_l_capsule, center=<-4.1534 7.3122 -0.2392>, rotation=<rot 0 0 90>}]
 foot_r [{shapeType=Capsule, length=14, radius=5.5, shapeName=foot_r_capsule, center=<4.1534 -7.3121 0.2392>, rotation=<rot 0 0 90>}]
 pelvis [{shapeType=Capsule, length=10, radius=9.25, shapeName=pelvis_capsule, center=<-0.5836 -2.2127 0>}]
 root [{shapeType=Capsule, length=60, radius=200, shapeName=root_capsule, center=<0 -21.8406 -194.3226>, rotation=<rot 0 90 90>}]
 spine_01 [{shapeType=Capsule, length=6.5, radius=9, shapeName=spine_01_capsule, center=<-0.3963 0.5179 0>, rotation=<rot 0 -90 0>}]
 spine_02 [{shapeType=Capsule, length=6.5, radius=8.5, shapeName=spine_02_capsule, center=<0 2.1926 -0.0166>, rotation=<rot 0.433 -90 0.0773>}]
 spine_03 [{shapeType=Capsule, length=5, radius=11, shapeName=spine_04_capsule, center=<-2 3.1779 0>, rotation=<rot 90 0 0>}]
 spine_04 [{shapeType=Capsule, length=6, radius=9.5, center=<-1.8727 3.0604 -0.0195>, rotation=<rot -0.2226 -90 -0.2664>}]
 spine_05 [
  {shapeType=Capsule, length=8, radius=8.25, shapeName=spine_05_capsule, center=<-2.5963 1.0241 0>}
  {shapeType=Capsule, length=9.8, radius=7.75, shapeName=spine_05_capsule, center=<-6.4845 6.1207 0>}
 ]
 thigh_l [
  {shapeType=Capsule, length=15, radius=8, shapeName=thigh_l_capsule, center=<-10 -1 1>, rotation=<rot 0 80 95>}
  {shapeType=Capsule, length=15, radius=7, shapeName=thigh_l_capsule, center=<-30 0 0>, rotation=<rot 90 0 0>}
 ]
 thigh_r [
  {shapeType=Capsule, length=15, radius=7.75, shapeName=thigh_r_capsule, center=<10 1 -1>, rotation=<rot 0 80 95>}
  {shapeType=Capsule, length=15, radius=6, shapeName=thigh_r_capsule, center=<30 0 0>, rotation=<rot 90 0 0>}
 ]
}
constraints [
 {bone1Name=spine_04, bone2Name=spine_02, swing1LimitDegrees=45, swing2LimitDegrees=45, twistLimitDegrees=45}
 {bone1Name=spine_05, bone2Name=spine_04, swing1LimitDegrees=45, swing2LimitDegrees=45, twistLimitDegrees=45}
 {bone1Name=thigh_l, bone2Name=pelvis, swing1LimitDegrees=45, swing2LimitDegrees=45, twistLimitDegrees=45}
 {bone1Name=calf_l, bone2Name=thigh_l, swing1LimitDegrees=45, swing2LimitDegrees=45, twistLimitDegrees=45}
 {bone1Name=dyn_skirtflap_l_01, bone2Name=dyn_skirtflap_l_MSTR, swing1LimitDegrees=25, swing2LimitDegrees=25, twistLimitDegrees=5}
 {bone1Name=dyn_skirtflap_l_03, bone2Name=dyn_skirtflap_l_02, swing1LimitDegrees=25, swing2LimitDegrees=25, twistLimitDegrees=5}
 {bone1Name=dyn_skirtflap_l_05, bone2Name=dyn_skirtflap_l_04, swing1LimitDegrees=25, swing2LimitDegrees=25, twistLimitDegrees=10}
 {bone1Name=thigh_r, bone2Name=pelvis, swing1LimitDegrees=45, swing2LimitDegrees=45, twistLimitDegrees=45}
 {bone1Name=calf_r, bone2Name=thigh_r, swing1LimitDegrees=45, swing2LimitDegrees=45, twistLimitDegrees=45}
 {bone1Name=dyn_skirtflap_r_01, bone2Name=dyn_skirtflap_r_MSTR, swing1LimitDegrees=25, swing2LimitDegrees=25, twistLimitDegrees=5}
 {bone1Name=dyn_skirtflap_r_03, bone2Name=dyn_skirtflap_r_02, swing1LimitDegrees=25, swing2LimitDegrees=25, twistLimitDegrees=5}
 {bone1Name=dyn_skirtflap_r_05, bone2Name=dyn_skirtflap_r_04, swing1LimitDegrees=25, swing2LimitDegrees=25, twistLimitDegrees=10}
 {bone1Name=dyn_bookstrap_01, bone2Name=dyn_bookstrap_MSTR, swing1LimitDegrees=25, swing2LimitDegrees=25, twistLimitDegrees=5}
 {bone1Name=dyn_bookstrap_02, bone2Name=dyn_bookstrap_01, swing1LimitDegrees=25, swing2LimitDegrees=25, twistLimitDegrees=5}
 {bone1Name=dyn_bookstrap_03, bone2Name=dyn_bookstrap_02, swing1LimitDegrees=25, swing2LimitDegrees=25, twistLimitDegrees=5}
 {
  bone1Name=dyn_bookstrap_MSTR
  bone2Name=pelvis
  swing1LimitDegrees=45
  swing1Motion=Free
  swing2LimitDegrees=45
  swing2Motion=Free
  twistLimitDegrees=45
  twistMotion=Free
 }
 {
  bone1Name=spine_03
  bone2Name=spine_02
  swing1LimitDegrees=45
  swing1Motion=Free
  swing2LimitDegrees=45
  swing2Motion=Free
  twistLimitDegrees=45
  twistMotion=Free
 }
 {
  bone1Name=spine_01
  bone2Name=pelvis
  swing1LimitDegrees=45
  swing1Motion=Free
  swing2LimitDegrees=45
  swing2Motion=Free
  twistLimitDegrees=45
  twistMotion=Free
 }
 {
  bone1Name=spine_02
  bone2Name=spine_01
  swing1LimitDegrees=45
  swing1Motion=Free
  swing2LimitDegrees=45
  swing2Motion=Free
  twistLimitDegrees=45
  twistMotion=Free
 }
 {
  bone1Name=foot_l
  bone2Name=calf_l
  swing1LimitDegrees=45
  swing1Motion=Free
  swing2LimitDegrees=45
  swing2Motion=Free
  twistLimitDegrees=45
  twistMotion=Free
 }
 {
  bone1Name=dyn_skirtflap_l_MSTR
  bone2Name=thigh_l
  swing1LimitDegrees=45
  swing1Motion=Free
  swing2LimitDegrees=45
  swing2Motion=Free
  twistLimitDegrees=45
  twistMotion=Free
 }
 {bone1Name=dyn_skirtflap_l_02, bone2Name=dyn_skirtflap_l_01, swing1LimitDegrees=25, swing2LimitDegrees=25, twistLimitDegrees=5}
 {bone1Name=dyn_skirtflap_l_04, bone2Name=dyn_skirtflap_l_03, swing1LimitDegrees=25, swing2LimitDegrees=25, twistLimitDegrees=10}
 {bone1Name=dyn_skirtflap_l_06, bone2Name=dyn_skirtflap_l_05, swing1LimitDegrees=25, swing2LimitDegrees=25, twistLimitDegrees=10}
 {
  bone1Name=foot_r
  bone2Name=calf_r
  swing1LimitDegrees=45
  swing1Motion=Free
  swing2LimitDegrees=45
  swing2Motion=Free
  twistLimitDegrees=45
  twistMotion=Free
 }
 {
  bone1Name=dyn_skirtflap_r_MSTR
  bone2Name=thigh_r
  swing1LimitDegrees=45
  swing1Motion=Free
  swing2LimitDegrees=45
  swing2Motion=Free
  twistLimitDegrees=45
  twistMotion=Free
 }
 {bone1Name=dyn_skirtflap_r_02, bone2Name=dyn_skirtflap_r_01, swing1LimitDegrees=25, swing2LimitDegrees=25, twistLimitDegrees=5}
 {bone1Name=dyn_skirtflap_r_04, bone2Name=dyn_skirtflap_r_03, swing1LimitDegrees=25, swing2LimitDegrees=25, twistLimitDegrees=10}
 {bone1Name=dyn_skirtflap_r_06, bone2Name=dyn_skirtflap_r_05, swing1LimitDegrees=25, swing2LimitDegrees=25, twistLimitDegrees=10}
]
