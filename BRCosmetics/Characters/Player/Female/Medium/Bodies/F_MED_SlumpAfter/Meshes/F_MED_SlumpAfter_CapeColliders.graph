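# Physics asset of cape collision volumes for the F_MED_SlumpAfter female medium player body cosmetic.
# Capsules cover pelvis, spine, clavicles and neck, with a sphere on the head; arm, leg and foot bodies are listed without shapes.
# Dynamic bones for the back capelet, skirt flaps and book strap carry their own capsules, with tiny ones on the capelet and flap MSTR roots.
# No constraints are defined, so the asset supplies collision shapes without joints.
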PhysicsAsset /BRCosmetics/Characters/Player/Female/Medium/Bodies/F_MED_SlumpAfter/Meshes/F_MED_SlumpAfter_CapeColliders
bodies {
 calf_l []
 calf_r []
 clavicle_l [{shapeType=Capsule, length=9.0192, radius=9.8871, center=<9.4266 -1.1806 -2.4977>, rotation=<rot 69.3297 1.4476 176.9462>}]
 clavicle_r [{shapeType=Capsule, length=9.0192, radius=9.8871, center=<-9.4266 1.1806 2.4977>, rotation=<rot 69.3297 1.4476 -3.0538>}]
 dyn_bookstrap_03 [{shapeType=Capsule, length=12, radius=12, shapeName=dyn_bookstrap_03_capsule, center=<7.5621 0 -2>, rotation=<rot 0 89.4626 -86.1591>}]
 dyn_bookstrap_MSTR [{shapeType=Capsule, length=8, radius=3.8485, shapeName=dyn_bookstrap_MSTR_capsule, center=<4.2184 0 0>, rotation=<rot 1.6153 -80.3205 -90>}]
 dyn_capelet_bk_c_01 [{shapeType=Capsule, length=8, radius=6, shapeName=dyn_capelet_bk_c_01_capsule, center=<8 0 -2>, rotation=<rot 0 90 90>}]
 dyn_capelet_bk_c_03 [{shapeType=Capsule, length=8, radius=6, shapeName=dyn_capelet_bk_c_03_capsule, center=<2 0 -4>, rotation=<rot 0 0 90>}]
 dyn_capelet_bk_c_MSTR [{shapeType=Capsule, length=0.505, radius=0.505, shapeName=dyn_capelet_bk_c_MSTR_capsule, rotation=<rot 0 -90 0>}]
 dyn_capelet_bk_l_01 [{shapeType=Capsule, length=8, radius=6, shapeName=dyn_capelet_bk_l_01_capsule, center=<8 0 -4>, rotation=<rot 0 90 90>}]
 dyn_capelet_bk_l_03 [{shapeType=Capsule, length=8, radius=6, shapeName=dyn_capelet_bk_l_03_capsule, center=<0 0 -4>, rotation=<rot 0 0 90>}]
 dyn_capelet_bk_l_MSTR [{shapeType=Capsule, length=0.505, radius=0.505, shapeName=dyn_capelet_bk_l_MSTR_capsule, rotation=<rot 0 -90 0>}]
 dyn_capelet_bk_r_01 [{shapeType=Capsule, length=8, radius=6, shapeName=dyn_capelet_bk_r_01_capsule, center=<8 0 -4>, rotation=<rot 0 90 90>}]
 dyn_capelet_bk_r_03 [{shapeType=Capsule, length=8, radius=6, shapeName=dyn_capelet_bk_r_03_capsule, center=<0 0 -4>, rotation=<rot 0 0 90>}]
 dyn_capelet_bk_r_MSTR [{shapeType=Capsule, length=0.505, radius=0.505, shapeName=dyn_capelet_bk_r_MSTR_capsule, rotation=<rot 0 -90 0>}]
 dyn_skirtflap_l_01 [{shapeType=Capsule, length=14.1416, radius=12, shapeName=dyn_skirtflap_l_01_capsule, center=<6 0 -8>, rotation=<rot 0 90 90>}]
 dyn_skirtflap_l_03 [{shapeType=Capsule, length=20.8445, radius=12, shapeName=dyn_skirtflap_l_03_capsule, center=<6 0 -8>, rotation=<rot 0 90 90>}]
 dyn_skirtflap_l_05 [{shapeType=Capsule, length=18.9522, radius=12, shapeName=dyn_skirtflap_l_05_capsule, center=<11.6514 0 -8>, rotation=<rot 0 43.8178 -81.7503>}]
 dyn_skirtflap_l_MSTR [{shapeType=Capsule, length=0.505, radius=0.505, shapeName=dyn_skirtflap_l_MSTR_capsule, rotation=<rot 0 -90 0>}]
 dyn_skirtflap_r_01 [{shapeType=Capsule, length=14.1411, radius=12, shapeName=dyn_skirtflap_r_01_capsule, center=<6 0 -8>, rotation=<rot 0 90 90>}]
 dyn_skirtflap_r_03 [{shapeType=Capsule, length=20.8445, radius=12, shapeName=dyn_skirtflap_r_03_capsule, center=<6 0 -8>, rotation=<rot 0 90 90>}]
 dyn_skirtflap_r_05 [{
   shapeType=Capsule
   length=18.9522
   radius=12
   shapeName=dyn_skirtflap_r_05_capsule
   center=<11.6514 0 -8>
   rotation=<rot 0.0001 136.1822 -76.7504>
  }]
 dyn_skirtflap_r_MSTR [{shapeType=Capsule, length=0.505, radius=0.505, shapeName=dyn_skirtflap_r_MSTR_capsule, rotation=<rot 0 -90 0>}]
 foot_l []
 foot_r []
 head [{shapeType=Sphere, radius=17.3125, center=<11.028 -0.174 0>}]
 lowerarm_l []
 lowerarm_r []
 neck_02 [{shapeType=Capsule, length=13.4375, radius=7.3175, center=<-0.2642 -0.3408 0>, rotation=<rot 90 -171.1193 -175.3978>}]
 pelvis [{shapeType=Capsule, length=8.6456, radius=14.5388, center=<-2.4887 0.1881 0>}]
 spine_01 [{shapeType=Capsule, length=7.9438, radius=13.5586, center=<-1.7965 0.1884 0>}]
 spine_02 [
  {shapeType=Capsule, length=7.3264, radius=13.6322, center=<-1.711 -0.1851 0.0001>}
  {shapeType=Capsule, length=29.3956, radius=11.5807, center=<2.5024 0.4992 3.9825>, rotation=<rot 90 0.0049 0.0049>}
  {shapeType=Capsule, length=29.875, radius=11.1013, center=<2.4704 0.4991 -3.9825>, rotation=<rot 90 -0.0396 -0.0396>}
 ]
 spine_03 [{shapeType=Capsule, length=8.2107, radius=13.7031, center=<0.1445 0.2396 0.0001>}]
 spine_04 [{shapeType=Capsule, length=9.4404, radius=14.1325, center=<2.766 0.7765 0.0002>}]
 spine_05 [
  {shapeType=Capsule, length=8.0226, radius=15.695, center=<-2.2795 1.8109 0>}
  {shapeType=Capsule, length=13.9518, radius=13.9425, center=<-0.2916 0.4452 0>}
 ]
 thigh_l []
 thigh_r []
 upperarm_l []
 upperarm_r []
}
constraints []
