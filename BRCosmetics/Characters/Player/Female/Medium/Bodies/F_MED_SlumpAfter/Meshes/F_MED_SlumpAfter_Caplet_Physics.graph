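# Physics asset for the capelet on the F_MED_SlumpAfter female body, setting up its dynamic bones for physics-driven motion.
# Capsules cover the pelvis, spine, neck, clavicles and arms, alongside seven three-segment capelet chains around front, sides and back.
# Each chain starts at a tiny MSTR capsule with free angular constraint to a clavicle, upper arm or spine_05.
# Chain joints act almost as hinges, with tight twist and first swing, wider second swing.
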PhysicsAsset /BRCosmetics/Characters/Player/Female/Medium/Bodies/F_MED_SlumpAfter/Meshes/F_MED_SlumpAfter_Caplet_Physics
bodies {
 clavicle_l [{shapeType=Capsule, length=12, radius=5.5, shapeName=clavicle_l_capsule, center=<3.5 -1 -1>, rotation=<rot 80 -0.0003 -0.0003>}]
 clavicle_r [{shapeType=Capsule, length=12, radius=5.5, shapeName=clavicle_r_capsule, center=<-3.5 1 1>, rotation=<rot 80 -0.0003 -0.0004>}]
 dyn_capelet_bk_c_01 [{shapeType=Capsule, length=20, radius=6, shapeName=dyn_capelet_bk_c_01_capsule, center=<6 0 6>, rotation=<rot 0 0 90>}]
 dyn_capelet_bk_c_02 [{shapeType=Capsule, length=20, radius=6, shapeName=dyn_capelet_bk_c_02_capsule, center=<6 0 6>, rotation=<rot 0 0 90>}]
 dyn_capelet_bk_c_03 [{shapeType=Capsule, length=20, radius=6, shapeName=dyn_capelet_bk_c_03_capsule, center=<6 0 6>, rotation=<rot -0.0004 0 90>}]
 dyn_capelet_bk_c_MSTR [{shapeType=Capsule, length=0.505, radius=0.505, shapeName=dyn_capelet_bk_c_MSTR_capsule, rotation=<rot 0 -90 0>}]
 dyn_capelet_bk_l_01 [{shapeType=Capsule, length=20, radius=6, center=<6 4 6>, rotation=<rot -0.0001 0 100>}]
 dyn_capelet_bk_l_02 [{shapeType=Capsule, length=20, radius=6, shapeName=dyn_capelet_bk_l_02_capsule, center=<6 4 6>, rotation=<rot -0.0001 0 100>}]
 dyn_capelet_bk_l_03 [{shapeType=Capsule, length=20, radius=6, shapeName=dyn_capelet_bk_l_03_capsule, center=<6 4 6>, rotation=<rot -0.0001 0 100>}]
 dyn_capelet_bk_l_MSTR [{shapeType=Capsule, length=0.505, radius=0.505, shapeName=dyn_capelet_bk_l_MSTR_capsule, rotation=<rot 0 -90 0>}]
 dyn_capelet_bk_r_01 [{shapeType=Capsule, length=20, radius=6, center=<6 -4 6>, rotation=<rot -0.0002 0 80.0001>}]
 dyn_capelet_bk_r_02 [{shapeType=Capsule, length=20, radius=6, shapeName=dyn_capelet_bk_r_02_capsule, center=<6 -4 6>, rotation=<rot -0.0002 0 80.0001>}]
 dyn_capelet_bk_r_03 [{shapeType=Capsule, length=20, radius=6, shapeName=dyn_capelet_bk_r_03_capsule, center=<6 -4 6>, rotation=<rot -0.0002 0 80.0001>}]
 dyn_capelet_bk_r_MSTR [{shapeType=Capsule, length=0.505, radius=0.505, shapeName=dyn_capelet_bk_r_MSTR_capsule, rotation=<rot 0 -90 0>}]
 dyn_capelet_ft_l_01 [{shapeType=Capsule, length=20, radius=6, shapeName=dyn_capelet_ft_l_01_capsule, center=<6 0 6>, rotation=<rot 0 0 90>}]
 dyn_capelet_ft_l_02 [{shapeType=Capsule, length=20, radius=6, shapeName=dyn_capelet_ft_l_02_capsule, center=<6 0 6>, rotation=<rot 0 0 90>}]
 dyn_capelet_ft_l_03 [{shapeType=Capsule, length=20, radius=6, shapeName=dyn_capelet_ft_l_03_capsule, center=<6 0 6>, rotation=<rot 0 0 90>}]
 dyn_capelet_ft_l_MSTR [{shapeType=Capsule, length=0.505, radius=0.505, shapeName=dyn_capelet_ft_l_MSTR_capsule, rotation=<rot 0 -90 0>}]
 dyn_capelet_ft_r_01 [{shapeType=Capsule, length=20, radius=6, shapeName=dyn_capelet_ft_r_01_capsule, center=<6 0 6>, rotation=<rot 0 0 90>}]
 dyn_capelet_ft_r_02 [{shapeType=Capsule, length=20, radius=6, shapeName=dyn_capelet_ft_r_02_capsule, center=<6 0 6>, rotation=<rot 0 0 90>}]
 dyn_capelet_ft_r_03 [{shapeType=Capsule, length=20, radius=6, shapeName=dyn_capelet_ft_r_03_capsule, center=<6 0 6>, rotation=<rot 0 0 90>}]
 dyn_capelet_ft_r_MSTR [{shapeType=Capsule, length=0.505, radius=0.505, shapeName=dyn_capelet_ft_r_MSTR_capsule, rotation=<rot 0 -90 0>}]
 dyn_capelet_l_01 [{shapeType=Capsule, length=20, radius=6, center=<6 0 6>, rotation=<rot 0 0 90>}]
 dyn_capelet_l_02 [{shapeType=Capsule, length=20, radius=6, shapeName=dyn_capelet_l_02_capsule, center=<6 0 6>, rotation=<rot 0 0 90>}]
 dyn_capelet_l_03 [{shapeType=Capsule, length=20, radius=6, shapeName=dyn_capelet_l_03_capsule, center=<6 0 6>, rotation=<rot 0 0 90>}]
 dyn_capelet_l_MSTR [{shapeType=Capsule, length=0.505, radius=0.505, shapeName=dyn_capelet_l_MSTR_capsule, rotation=<rot 0 -90 0>}]
 dyn_capelet_r_01 [{shapeType=Capsule, length=20, radius=6, center=<6 0 6>, rotation=<rot 0 0 90>}]
 dyn_capelet_r_02 [{shapeType=Capsule, length=20, radius=6, shapeName=dyn_capelet_r_02_capsule, center=<6 0 6>, rotation=<rot 0 0 90>}]
 dyn_capelet_r_03 [{shapeType=Capsule, length=20, radius=6, shapeName=dyn_capelet_r_03_capsule, center=<6 0 6>, rotation=<rot 0 0 90>}]
 dyn_capelet_r_MSTR [{shapeType=Capsule, length=0.505, radius=0.505, shapeName=dyn_capelet_r_MSTR_capsule, rotation=<rot 0 -90 0>}]
 lowerarm_l [{shapeType=Capsule, length=16, radius=6, center=<9.5 0 0>, rotation=<rot -90 0 0>}]
 lowerarm_r [{shapeType=Capsule, length=16, radius=6, center=<-9.5 0 0>, rotation=<rot -90 0 0>}]
 neck_01 [{shapeType=Capsule, length=8, radius=4.5, shapeName=neck_01_capsule, center=<3.6683 -0.2876 0>, rotation=<rot 90 0 0>}]
 pelvis [{shapeType=Capsule, length=10, radius=9.25, shapeName=pelvis_capsule, center=<-0.5836 -2.2127 0>}]
 spine_01 [{shapeType=Capsule, length=6.5, radius=9, shapeName=spine_01_capsule, center=<-0.3963 0.5179 0>, rotation=<rot 0 -90 0>}]
 spine_02 [{shapeType=Capsule, length=6.5, radius=8.5, shapeName=spine_02_capsule, center=<0 2.1926 -0.0166>, rotation=<rot 0.433 -90 0.0773>}]
 spine_03 [{shapeType=Capsule, length=5, radius=11, shapeName=spine_04_capsule, center=<-2 3.1779 0>, rotation=<rot 90 0 0>}]
 spine_04 [{shapeType=Capsule, length=6, radius=9.5, center=<-1.8727 3.0604 -0.0195>, rotation=<rot -0.2226 -90 -0.2664>}]
 spine_05 [
  {shapeType=Capsule, length=8, radius=8.25, shapeName=spine_05_capsule, center=<-2.5963 1.0241 0>}
  {shapeType=Capsule, length=9.8, radius=7.75, shapeName=spine_05_capsule, center=<-6.4845 6.1207 0>}
 ]
 upperarm_l [{shapeType=Capsule, length=20, radius=4.5, shapeName=upperarm_l_capsule, center=<12 0 0>, rotation=<rot 90 0 0>}]
 upperarm_r [{shapeType=Capsule, length=20, radius=4.5, shapeName=upperarm_r_capsule, center=<-12 0 0>, rotation=<rot 90 0 0>}]
}
constraints [
 {bone1Name=spine_04, bone2Name=spine_02, swing1LimitDegrees=45, swing2LimitDegrees=45, twistLimitDegrees=45}
 {bone1Name=spine_05, bone2Name=spine_04, swing1LimitDegrees=45, swing2LimitDegrees=45, twistLimitDegrees=45}
 {bone1Name=clavicle_l, bone2Name=spine_05, swing1LimitDegrees=45, swing2LimitDegrees=45, twistLimitDegrees=45}
 {bone1Name=lowerarm_l, bone2Name=clavicle_l, swing1LimitDegrees=45, swing2LimitDegrees=45, twistLimitDegrees=45}
 {bone1Name=dyn_capelet_l_01, bone2Name=dyn_capelet_l_MSTR, swing1LimitDegrees=1, swing2LimitDegrees=22, twistLimitDegrees=1}
 {bone1Name=dyn_capelet_bk_l_01, bone2Name=dyn_capelet_bk_l_MSTR, swing1LimitDegrees=1, swing2LimitDegrees=22, twistLimitDegrees=1}
 {bone1Name=clavicle_r, bone2Name=spine_05, swing1LimitDegrees=45, swing2LimitDegrees=45, twistLimitDegrees=45}
 {bone1Name=lowerarm_r, bone2Name=clavicle_r, swing1LimitDegrees=45, swing2LimitDegrees=45, twistLimitDegrees=45}
 {bone1Name=dyn_capelet_r_01, bone2Name=dyn_capelet_r_MSTR, swing1LimitDegrees=1, swing2LimitDegrees=22, twistLimitDegrees=1}
 {bone1Name=dyn_capelet_bk_r_01, bone2Name=dyn_capelet_bk_r_MSTR, swing1LimitDegrees=1, swing2LimitDegrees=22, twistLimitDegrees=1}
 {
  bone1Name=spine_03
  bone2Name=spine_02
  swing1LimitDegrees=45
  swing1Motion=Free
  swing2LimitDegrees=45
  swing2Motion=Free
  twistLimitDegrees=45
  twistMotion=Free
 }
 {
  bone1Name=neck_01
  bone2Name=spine_02
  swing1LimitDegrees=45
  swing1Motion=Free
  swing2LimitDegrees=45
  swing2Motion=Free
  twistLimitDegrees=45
  twistMotion=Free
 }
 {
  bone1Name=upperarm_l
  bone2Name=spine_02
  swing1LimitDegrees=45
  swing1Motion=Free
  swing2LimitDegrees=45
  swing2Motion=Free
  twistLimitDegrees=45
  twistMotion=Free
 }
 {
  bone1Name=upperarm_r
  bone2Name=spine_02
  swing1LimitDegrees=45
  swing1Motion=Free
  swing2LimitDegrees=45
  swing2Motion=Free
  twistLimitDegrees=45
  twistMotion=Free
 }
 {
  bone1Name=spine_01
  bone2Name=pelvis
  swing1LimitDegrees=45
  swing1Motion=Free
  swing2LimitDegrees=45
  swing2Motion=Free
  twistLimitDegrees=45
  twistMotion=Free
 }
 {
  bone1Name=spine_02
  bone2Name=spine_01
  swing1LimitDegrees=45
  swing1Motion=Free
  swing2LimitDegrees=45
  swing2Motion=Free
  twistLimitDegrees=45
  twistMotion=Free
 }
 {
  bone1Name=dyn_capelet_bk_l_MSTR
  bone2Name=clavicle_l
  swing1LimitDegrees=45
  swing1Motion=Free
  swing2LimitDegrees=45
  swing2Motion=Free
  twistLimitDegrees=45
  twistMotion=Free
 }
 {bone1Name=dyn_capelet_bk_l_02, bone2Name=dyn_capelet_bk_l_01, swing1LimitDegrees=1, swing2LimitDegrees=22, twistLimitDegrees=1}
 {bone1Name=dyn_capelet_bk_l_03, bone2Name=dyn_capelet_bk_l_02, swing1LimitDegrees=1, swing2LimitDegrees=22, twistLimitDegrees=1}
 {
  bone1Name=dyn_capelet_ft_l_MSTR
  bone2Name=clavicle_l
  swing1LimitDegrees=45
  swing1Motion=Free
  swing2LimitDegrees=45
  swing2Motion=Free
  twistLimitDegrees=45
  twistMotion=Free
 }
 {bone1Name=dyn_capelet_ft_l_01, bone2Name=dyn_capelet_ft_l_MSTR, swing1LimitDegrees=5, swing2LimitDegrees=25, twistLimitDegrees=1}
 {bone1Name=dyn_capelet_ft_l_02, bone2Name=dyn_capelet_ft_l_01, swing1LimitDegrees=5, swing2LimitDegrees=25, twistLimitDegrees=1}
 {bone1Name=dyn_capelet_ft_l_03, bone2Name=dyn_capelet_ft_l_02, swing1LimitDegrees=5, swing2LimitDegrees=25, twistLimitDegrees=1}
 {
  bone1Name=dyn_capelet_l_MSTR
  bone2Name=upperarm_l
  swing1LimitDegrees=45
  swing1Motion=Free
  swing2LimitDegrees=45
  swing2Motion=Free
  twistLimitDegrees=45
  twistMotion=Free
 }
 {bone1Name=dyn_capelet_l_02, bone2Name=dyn_capelet_l_01, swing1LimitDegrees=1, swing2LimitDegrees=22, twistLimitDegrees=1}
 {bone1Name=dyn_capelet_l_03, bone2Name=dyn_capelet_l_02, swing1LimitDegrees=1, swing2LimitDegrees=22, twistLimitDegrees=1}
 {
  bone1Name=dyn_capelet_bk_r_MSTR
  bone2Name=clavicle_r
  swing1LimitDegrees=45
  swing1Motion=Free
  swing2LimitDegrees=45
  swing2Motion=Free
  twistLimitDegrees=45
  twistMotion=Free
 }
 {bone1Name=dyn_capelet_bk_r_02, bone2Name=dyn_capelet_bk_r_01, swing1LimitDegrees=1, swing2LimitDegrees=22, twistLimitDegrees=1}
 {bone1Name=dyn_capelet_bk_r_03, bone2Name=dyn_capelet_bk_r_02, swing1LimitDegrees=1, swing2LimitDegrees=22, twistLimitDegrees=1}
 {bone1Name=dyn_capelet_ft_r_01, bone2Name=dyn_capelet_ft_r_MSTR, swing1LimitDegrees=5, swing2LimitDegrees=25, twistLimitDegrees=1}
 {
  bone1Name=dyn_capelet_ft_r_MSTR
  bone2Name=clavicle_r
  swing1LimitDegrees=45
  swing1Motion=Free
  swing2LimitDegrees=45
  swing2Motion=Free
  twistLimitDegrees=45
  twistMotion=Free
 }
 {bone1Name=dyn_capelet_ft_r_02, bone2Name=dyn_capelet_ft_r_01, swing1LimitDegrees=5, swing2LimitDegrees=25, twistLimitDegrees=1}
 {bone1Name=dyn_capelet_ft_r_03, bone2Name=dyn_capelet_ft_r_02, swing1LimitDegrees=5, swing2LimitDegrees=25, twistLimitDegrees=1}
 {
  bone1Name=dyn_capelet_r_MSTR
  bone2Name=upperarm_r
  swing1LimitDegrees=45
  swing1Motion=Free
  swing2LimitDegrees=45
  swing2Motion=Free
  twistLimitDegrees=45
  twistMotion=Free
 }
 {bone1Name=dyn_capelet_r_02, bone2Name=dyn_capelet_r_01, swing1LimitDegrees=1, swing2LimitDegrees=22, twistLimitDegrees=1}
 {bone1Name=dyn_capelet_r_03, bone2Name=dyn_capelet_r_02, swing1LimitDegrees=1, swing2LimitDegrees=22, twistLimitDegrees=1}
 {
  bone1Name=dyn_capelet_bk_c_MSTR
  bone2Name=spine_05
  swing1LimitDegrees=45
  swing1Motion=Free
  swing2LimitDegrees=45
  swing2Motion=Free
  twistLimitDegrees=45
  twistMotion=Free
 }
 {bone1Name=dyn_capelet_bk_c_01, bone2Name=dyn_capelet_bk_c_MSTR, swing1LimitDegrees=1, swing2LimitDegrees=22, twistLimitDegrees=1}
 {bone1Name=dyn_capelet_bk_c_02, bone2Name=dyn_capelet_bk_c_01, swing1LimitDegrees=1, swing2LimitDegrees=22, twistLimitDegrees=1}
 {bone1Name=dyn_capelet_bk_c_03, bone2Name=dyn_capelet_bk_c_02, swing1LimitDegrees=1, swing2LimitDegrees=22, twistLimitDegrees=1}
]
